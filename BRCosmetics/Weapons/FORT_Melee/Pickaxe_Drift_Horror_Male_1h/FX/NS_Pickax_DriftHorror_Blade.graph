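# Niagara system for the Drift Horror pickaxe blade, producing intermittent magenta lightning flickers with lights as short bursts on randomized loop timers.
# Spike_Arcs picks socket sets from the user skeletal mesh via a custom module, re-randomizes endpoints each update, and stretches meshes between them.
# Out_of_spikes uses the same sampling to shoot scaled bolt meshes along socket directions.
# On_ring scatters randomly oriented sprites on a cylinder; the component-rendered Ref emitter is disabled.
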NiagaraSystem /BRCosmetics/Weapons/FORT_Melee/Pickaxe_Drift_Horror_Male_1h/FX/NS_Pickax_DriftHorror_Blade
dependencies {usedDataInterfaces [&NiagaraDataInterfaceCurve], usedDynamicInputs [], usedModules [&SystemState]}
emitters {
 Lightning_Spike_Arcs {
  emitterUpdateScript [
   &EmitterState(
    Life Cycle Mode=::NewEnumerator1
    Loop Duration=&RandomRangeFloat(Minimum=0.1, Maximum=3)
    Recalculate Duration Each Loop=true
    UseLoopDelay=false
   )
   &SpawnBurst_Instantaneous(
    Spawn Count=&MakeCustomIntFromBool(
     Bool=&SetBoolByFloatComparison(A=&RandomRangeFloat(Minimum=0, Maximum=1), B=0.25)
     True INT=1
     False INT=2
    )
    Use Spawn Probability=false
    UseLoopCountLimit=false
   )
  ]
  particleSpawnScript [
   &InitializeParticle(
    Color=<rgba 15 0 11.3846 1>
    Lifetime=0.1
    Mesh Scale Max=<0.037 0.5 0.5>
    Mesh Scale Min=<0.037 1 1>
    Mesh Scale Mode=::NewEnumerator2
    Position Mode=::NewEnumerator1
    Sprite Size Mode=::NewEnumerator3
    Uniform Sprite Size=2
    UsePositionOffset=false
    Write Material Random=true
   )
   /SetVariables(Particles.Pos Choice:NiagaraInt32=&Multiply_Int(A=&RandomRangeInt(Minimum=0, Maximum=2), B=2)) {isSetParameters=true, module=SetVariables_#1}
   &InitialMeshOrientation(
    Mesh Orientation Mode=::NewEnumerator7
    Rotation=&RandomRangeVector(Minimum=<0 0 0>, Maximum=<1 0 0>)
    Use Rotation Vector=true
   )
  ]
  particleUpdateScript [
   &ParticleState()
   &self:Drift_Horror_Get_Set_of_4_Sockets(New Skeletal Mesh:NiagaraDataInterfaceSkeletalMesh=$User.SKeletalMesh, StartIndex:NiagaraInt32=$Particles.Initial.Pos Choice) {module=Drift_Horror_Get_Set_of_4_Sockets}
   /SetVariables(
    Particles.Pos A:Vector3f=&Lerp_Vector(
     A=$Output.Drift_Horror_Get_Set_of_4_Sockets.BonePos1
     B=$Output.Drift_Horror_Get_Set_of_4_Sockets.BonePos2
     Alpha=&RandomRangeFloat(Minimum=0, Maximum=1)
    )
    Particles.Pos B:Vector3f=&Lerp_Vector(
     A=$Output.Drift_Horror_Get_Set_of_4_Sockets.BonePos3
     B=$Output.Drift_Horror_Get_Set_of_4_Sockets.BonePos4
     Alpha=&RandomRangeFloat(Minimum=0, Maximum=1)
    )
   ) {isSetParameters=true, module=SetVariables_#2}
   /SetVariables(Particles.Position:Vector3f=$Particles.Pos A) {isSetParameters=true, module=SetVariables_#3}
   &DynamicMaterialParameters(
    Index 0 Param 1=&RandomRangeFloat(Minimum=-2, Maximum=-1)
    Index 0 Param 2=1
    Index 0 Param 3=-0.1
    Module.Param0WriteEnabled=true
    Module.Param1WriteEnabled=true
    Module.Param2WriteEnabled=true
    Module.Param3WriteEnabled=true
   )
   /Niagara/Modules/Update/Orientation/V2/OrientMeshToVector(
    Look At Direction:Vector3f=&Subtract_Vector(A=$Particles.Pos B, B=$Particles.Pos A)
    Override Quat=false
    Up Axis Direction:Vector3f=<0 0 1>
    UseOptionalSideAxis=false
    UseOptionalUpAxis=false
   )
   &ScaleMeshSize(Scale Factor=&MakeVector(
     X=&VectorLength(Vector=&Subtract_Vector(A=$Particles.Pos B, B=$Particles.Pos A))
     Y=1
     Z=1
    ))
   &FN_Light_Attributes(
    Color=<rgba 5 1.25 4.0961 1>
    Create Color Attribute=true
    Create Exponent Attribute=true
    Create Position Attribute=true
    Position=&Lerp_Vector(A=$Particles.Pos A, B=$Particles.Pos B, Alpha=0.5)
    Scale Color=&FloatFromCurve(FloatCurve=DataInterface() {propertyValues {Curve {keys [
         {arriveTangent=5.7658, interpMode=RCIM_Cubic, leaveTangent=5.7658, tangentMode=RCTM_User, time=0, value=0}
         <at 0.3 0.9>
         {interpMode=RCIM_Cubic, tangentMode=RCTM_User, time=0.5, value=0.5}
         <at 0.8 0.8>
         {arriveTangent=-4.8787, interpMode=RCIM_Cubic, leaveTangent=-4.8787, tangentMode=RCTM_User, time=1, value=0}
        ]}}})
   )
  ]
  rendererClasses [&NiagaraMeshRendererProperties, &NiagaraLightRendererProperties]
  renderers [
   {rendererClass=&NiagaraMeshRendererProperties, rendererIndex=0}
   {rendererClass=&NiagaraLightRendererProperties, rendererIndex=1}
  ]
 }
 Lightning_on_ring {
  emitterUpdateScript [
   &EmitterState(
    Life Cycle Mode=::NewEnumerator1
    Loop Duration=&RandomRangeFloat(Minimum=0.1, Maximum=3)
    Recalculate Duration Each Loop=true
    UseLoopDelay=false
   )
   &SpawnBurst_Instantaneous(
    Spawn Count=&RandomRangeInt(Minimum=1, Maximum=3)
    Use Spawn Probability=false
    UseLoopCountLimit=false
   )
  ]
  particleSpawnScript [
   &InitializeParticle(
    Color=<rgba 25 0 18.9743 1>
    Lifetime Max=0.25
    Lifetime Min=0.05
    Lifetime Mode=::NewEnumerator1
    Sprite Size Max=<10 12>
    Sprite Size Min=<5 3>
    Sprite Size Mode=::NewEnumerator2
    Sprite UV Mode=::NewEnumerator1
    UsePositionOffset=false
    Write Material Random=true
   )
   &CylinderLocation(
    Clamp UV Range=false
    Cylinder Axis:Vector3f=<0 1 0>
    Cylinder Distribution:ENiagara_CylinderMode=::NewEnumerator2
    Cylinder Height=1
    Cylinder Radius=13
    MaskSpawnGroup=false
    Offset=<4 0 -2.5>
    Offset Mode:ENiagara_OffsetMode=::NewEnumerator0
    Rotation Mode:ENiagara_RotationMode=::NewEnumerator1
    U Position:NiagaraFloat=&RandomRangeFloat(Minimum=0.2, Maximum=0.85)
    UseOwnerScale=true
    V Position:NiagaraFloat=&RandomRangeFloat(Minimum=0, Maximum=1)
   )
   &SpriteFacingAndAlignment(
    Sprite Alignment=&RandomRangeVector(Minimum=<-1 -1 -1>, Maximum=<1 1 1>)
    Sprite Facing=&RandomRangeVector(Minimum=<-1 -1 -1>, Maximum=<1 1 1>)
    WriteAlignment=true
    WriteFacing=true
   )
  ]
  particleUpdateScript [
   &ParticleState()
   &FN_Light_Attributes(
    Color=<rgba 50 12.5 40.9613 1>
    Create Color Attribute=true
    Create Exponent Attribute=true
    Create Position Attribute=true
    Position:NiagaraPosition=$Particles.Position
   )
  ]
  rendererClasses [&NiagaraSpriteRendererProperties, &NiagaraLightRendererProperties]
  renderers [
   {rendererClass=&NiagaraSpriteRendererProperties}
   {rendererClass=&NiagaraLightRendererProperties, rendererIndex=1}
  ]
 }
 Lightning_ou_of_spikes {
  emitterUpdateScript [
   &EmitterState(
    Life Cycle Mode=::NewEnumerator1
    Loop Duration=&RandomRangeFloat(Minimum=0.1, Maximum=3)
    Recalculate Duration Each Loop=true
    UseLoopDelay=false
   )
   &SpawnBurst_Instantaneous(
    Spawn Count=&MakeCustomIntFromBool(
     Bool=&SetBoolByFloatComparison(A=&RandomRangeFloat(Minimum=0, Maximum=1), B=0.25)
     True INT=1
     False INT=2
    )
    Use Spawn Probability=false
    UseLoopCountLimit=false
   )
  ]
  particleSpawnScript [
   &InitializeParticle(
    Color=<rgba 15 0 11.3846 1>
    Lifetime=0.1
    Mesh Scale Mode=::NewEnumerator3
    Mesh Uniform Scale=0.5
    Position Mode=::NewEnumerator1
    Sprite Size Mode=::NewEnumerator3
    Uniform Sprite Size=2
    UsePositionOffset=false
    Write Material Random=true
   )
   /SetVariables(Particles.Pos Choice:NiagaraInt32=&Multiply_Int(A=&RandomRangeInt(Minimum=0, Maximum=4), B=2)) {isSetParameters=true, module=SetVariables_#1}
   &InitialMeshOrientation(
    Mesh Orientation Mode=::NewEnumerator7
    Rotation=&RandomRangeVector(Minimum=<0 0.25 0.25>, Maximum=<1 -0.25 -0.25>)
    Use Rotation Vector=true
   )
  ]
  particleUpdateScript [
   &ParticleState()
   &self:Drift_Horror_Get_Set_of_4_Sockets(New Skeletal Mesh:NiagaraDataInterfaceSkeletalMesh=$User.SKeletalMesh, StartIndex:NiagaraInt32=$Particles.Initial.Pos Choice) {module=Drift_Horror_Get_Set_of_4_Sockets}
   /SetVariables(
    Particles.Pos A:Vector3f=&Lerp_Vector(
     A=$Output.Drift_Horror_Get_Set_of_4_Sockets.BonePos1
     B=$Output.Drift_Horror_Get_Set_of_4_Sockets.BonePos2
     Alpha=&RandomRangeFloat(Minimum=0, Maximum=1)
    )
    Particles.Pos B:Vector3f=&Lerp_Vector(
     A=$Output.Drift_Horror_Get_Set_of_4_Sockets.BonePos3
     B=$Output.Drift_Horror_Get_Set_of_4_Sockets.BonePos4
     Alpha=&RandomRangeFloat(Minimum=0, Maximum=1)
    )
   ) {isSetParameters=true, module=SetVariables_#2}
   /SetVariables(Particles.Position:Vector3f=$Particles.Pos A) {isSetParameters=true, module=SetVariables_#3}
   &OrientMeshToVector(
    Look At Direction:Vector3f=&Subtract_Vector(A=$Output.Drift_Horror_Get_Set_of_4_Sockets.BonePos2, B=$Output.Drift_Horror_Get_Set_of_4_Sockets.BonePos1)
    Override Quat=false
    Up Axis Direction:Vector3f=<0 0 1>
    UseOptionalSideAxis=false
    UseOptionalUpAxis=false
   )
   &DynamicMaterialParameters(
    Index 0 Param 1=1
    Index 0 Param 2=1
    Index 0 Param 3=-0.1
    Module.Param0WriteEnabled=true
    Module.Param1WriteEnabled=true
    Module.Param2WriteEnabled=true
    Module.Param3WriteEnabled=true
   )
   &ScaleMeshSize(Scale Factor=&MakeVector(
     X=&FloatFromCurve(FloatCurve=DataInterface() {propertyValues {Curve {keys [{tangentMode=RCTM_User, time=0, value=0.5}, {tangentMode=RCTM_User, time=1, value=1}]}}})
     Y=&FloatFromCurve(FloatCurve=DataInterface() {propertyValues {Curve {keys [{tangentMode=RCTM_User, time=0, value=1}, {tangentMode=RCTM_User, time=1, value=0.5}]}}})
     Z=&FloatFromCurve(FloatCurve=DataInterface() {propertyValues {Curve {keys [{tangentMode=RCTM_User, time=0, value=1}, {tangentMode=RCTM_User, time=1, value=0.5}]}}})
    ))
   &FN_Light_Attributes(
    Color=<rgba 5 1.25 4.0961 1>
    Create Color Attribute=true
    Create Exponent Attribute=true
    Create Position Attribute=true
    Position=&Add_Vector(
     A=$Particles.Pos A
     B=&Multiply_VectorByFloat(
      Vector=&Subtract_Vector(A=$Output.Drift_Horror_Get_Set_of_4_Sockets.BonePos2, B=$Output.Drift_Horror_Get_Set_of_4_Sockets.BonePos1)
      Float=0.5
     )
    )
   )
  ]
  rendererClasses [&NiagaraMeshRendererProperties, &NiagaraLightRendererProperties]
  renderers [
   {rendererClass=&NiagaraMeshRendererProperties, rendererIndex=0}
   {rendererClass=&NiagaraLightRendererProperties, rendererIndex=1}
  ]
 }
 Ref {
  bEnabled=false
  emitterUpdateScript [
   &EmitterState()
   &SpawnBurst_Instantaneous(Spawn Time=0.01, Use Spawn Probability=false, UseLoopCountLimit=false)
  ]
  particleSpawnScript [&InitializeParticle(Color Mode=::NewEnumerator0, Lifetime=5, UsePositionOffset=false, Write Material Random=true)]
  particleUpdateScript [&ParticleState()]
  rendererClasses [&NiagaraComponentRendererProperties]
  renderers [{rendererClass=&NiagaraComponentRendererProperties, rendererIndex=0}]
 }
}
summary {
 systemName=NS_Pickax_DriftHorror_Blade
 emitters [
  {
   emitterName=Lightning_Spike_Arcs
   rendererClasses [&NiagaraMeshRendererProperties, &NiagaraLightRendererProperties]
  }
  {bEnabled=false, emitterName=Ref, rendererClasses [&NiagaraComponentRendererProperties]}
  {
   emitterName=Lightning_ou_of_spikes
   rendererClasses [&NiagaraMeshRendererProperties, &NiagaraLightRendererProperties]
  }
  {
   emitterName=Lightning_on_ring
   rendererClasses [&NiagaraSpriteRendererProperties, &NiagaraLightRendererProperties]
  }
 ]
 userVariables [User.SkeletalMesh {
   type=NiagaraDataInterfaceSkeletalMesh
   defaultValue {
    struct=NiagaraExt_VariableValue_DataInterface
    value {dataInterface=&self:NiagaraDataInterfaceSkeletalMesh_0}
   }
  }]
}
